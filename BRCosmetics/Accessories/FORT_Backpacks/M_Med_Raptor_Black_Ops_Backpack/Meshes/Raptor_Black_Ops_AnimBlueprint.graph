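# Anim Blueprint for the Raptor Black Ops backpack that adds physical secondary motion to its four bottle bones.
# Starting from the ref pose, four chained Anim Dynamics nodes simulate the left and right bottles with overridden gravity, damping and tight angular limits.
# All four share DynAlpha, which each update sets low while bIsTargeting is true to damp the sway, and back to half strength otherwise.
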AnimBlueprint /BRCosmetics/Accessories/FORT_Backpacks/M_Med_Raptor_Black_Ops_Backpack/Meshes/Raptor_Black_Ops_AnimBlueprint
target_skeleton=&/Raptor_Black_Ops_Skeleton
graphs [
 AnimGraph {nodes [
   n1 = Get DynAlpha() -> (DynAlpha:double)
   n2 = Anim Dynamics - Bone: dyn_l_bottle_01(ComponentPose=Local To Component(LocalPose=Local Space Ref Pose()), Alpha=@n1, ExternalForce="0,0,0") {
    alpha=0.5
    angularDampingOverride=0.997
    bGravityOverrideInSimSpace=true
    bOverrideAngularDamping=true
    bUseGravityOverride=true
    alphaBoolBlend {blendOption=Linear}
    boundBone {boneName=dyn_l_bottle_01}
    componentAppliedLinearAccClamp=<1000 1000 1000>
    componentLinearAccScale=<0.25 0.25 0.25>
    gravityOverride {z=-980}
    physicsBodyDefinitions [{
      sphereCollisionRadius=0
      boundBone {boneName=dyn_l_bottle_01}
      boxExtents=<8 4 4>
      constraintSetup {angularLimitsMax=<10 5 10>, angularLimitsMin=<-10 -5 -10>, angularTarget=<0 0 0>}
      localJointOffset=<-1 0 0.1>
     }]
   }
   n3 = Anim Dynamics - Bone: dyn_l_bottle_02(ComponentPose=@n2, Alpha=@n1, ExternalForce="0,0,0") {
    alpha=0.5
    angularDampingOverride=0.995
    bGravityOverrideInSimSpace=true
    bOverrideAngularDamping=true
    bUseGravityOverride=true
    alphaBoolBlend {blendOption=Linear}
    boundBone {boneName=dyn_l_bottle_02}
    componentAppliedLinearAccClamp=<1500 1500 1500>
    componentLinearAccScale=<0.3 0.3 0.3>
    componentLinearVelScale=<2 2 2>
    gravityOverride {z=-980}
    physicsBodyDefinitions [{
      sphereCollisionRadius=0
      boundBone {boneName=dyn_l_bottle_02}
      boxExtents=<8 4 4>
      constraintSetup {angularLimitsMax=<5 5 15>, angularLimitsMin=<-5 -5 -15>, angularTarget=<0 0 0>}
      localJointOffset=<-0.9 0 0>
     }]
   }
   n4 = Anim Dynamics - Bone: dyn_r_bottle_01(ComponentPose=@n3, Alpha=@n1, ExternalForce="0,0,0") {
    alpha=0.5
    angularDampingOverride=0.995
    bGravityOverrideInSimSpace=true
    bOverrideAngularDamping=true
    bUseGravityOverride=true
    alphaBoolBlend {blendOption=Linear}
    boundBone {boneName=dyn_r_bottle_01}
    componentAppliedLinearAccClamp=<1500 1500 1500>
    componentLinearAccScale=<0.3 0.3 0.3>
    componentLinearVelScale=<2 2 2>
    gravityOverride {z=-980}
    physicsBodyDefinitions [{
      sphereCollisionRadius=0
      boundBone {boneName=dyn_r_bottle_01}
      boxExtents=<8 4 4>
      constraintSetup {angularLimitsMax=<5 5 15>, angularLimitsMin=<-5 -5 -15>, angularTarget=<0 0 0>}
      localJointOffset=<-0.9 0 0>
     }]
   }
   n5 = Anim Dynamics - Bone: dyn_r_bottle_02(ComponentPose=@n4, Alpha=@n1, ExternalForce="0,0,0") {
    alpha=0.5
    angularDampingOverride=0.997
    bGravityOverrideInSimSpace=true
    bOverrideAngularDamping=true
    bUseGravityOverride=true
    alphaBoolBlend {blendOption=Linear}
    boundBone {boneName=dyn_r_bottle_02}
    componentAppliedLinearAccClamp=<1000 1000 1000>
    componentLinearAccScale=<0.25 0.25 0.25>
    gravityOverride {z=-980}
    physicsBodyDefinitions [{
      sphereCollisionRadius=0
      boundBone {boneName=dyn_r_bottle_02}
      boxExtents=<8 4 4>
      constraintSetup {angularLimitsMax=<5 5 15>, angularLimitsMin=<-5 -5 -15>, angularTarget=<0 0 0>}
      localJointOffset=<-1 0 0.1>
     }]
   }
   Output Pose(Result=Component To Local(ComponentPose=@n5)) {layerGroup=None}
  ]}
 EventGraph {nodes [
   Event BlueprintUpdateAnimation()
   > n1 = Branch(Condition=Get bIsTargeting())
   > Set DynAlpha(DynAlpha:double=0.1) -> (Output_Get:double)
   Set DynAlpha(execute=@n1.else, DynAlpha=0.5)
   TryGetPawnOwner()
  ]}
]
variables [DynAlpha {type=number, default=0.5}]
